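; Building gameplay actor that manages health, damage feedback and destruction for the hot air balloon.
; On begin play it waits for the playlist, then applies a GameplayEffect levelled by clamped team size to scale health.
; An optional authority-only looping timer applies storm damage outside the safe zone and removes it inside.
; Damage shows a timed health bar and plays hitmarker audio for the locally viewed instigator; death fires a replicated destruction GameplayCue.
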(event EventBeginPlay
  (EventDispatchers|BindEventtoOnDamaged (delegate Custom|OnDamaged_Event))
  (EventDispatchers|BindEventtoOnDied (delegate Custom|OnDied_Event))
  (CallFunction|SetHealthbyPlaylist)
  (if (ScalableFloat|GetValueasBool (Variables|Default|GetRow_ShouldTakeStormDamage))
    (Networking|SwitchHasAuthority
      (:Authority
        (CallFunction|StartStormChecking)))))

(event Custom|OnDied_Event
  (params (DamagedActor :type Actor%20Object%20Reference) (Damage :type Float%20%28single-precision%29) (InstigatedBy :type Controller%20Object%20Reference) (DamageCauser :type Actor%20Object%20Reference) (HitLocation :type Vector) (FHitComponent :type Primitive%20Component%20Object%20Reference) (BoneName :type Name) (Momentum :type Vector))
  (Utilities|Time|ClearandInvalidateTimerbyHandle (Variables|Default|GetHealthBarTimerHandle))
  (Ability|ExecuteGameplayCueReplicated (Variables|Default|GetGCN_DestructionTag)))

(event Custom|OnDamaged_Event
  (params (DamagedActor :type Actor%20Object%20Reference) (Damage :type Float%20%28single-precision%29) (InstigatedBy :type Controller%20Object%20Reference) (DamageCauser :type Actor%20Object%20Reference) (HitLocation :type Vector) (FHitComponent :type Primitive%20Component%20Object%20Reference) (BoneName :type Name) (Momentum :type Vector))
  (HealthBarComponent|SetHealthBarVisibility (Variables|Default|GetFortHealthBar) true)
  (bind _returnvalue (Utilities|Time|SetTimerbyEvent (delegate Custom|HideHealthBar) (ScalableFloat|GetValueatLevel (Variables|Default|GetRow_HealthBarActiveDuration))))
  (Variables|Default|SetHealthBarTimerHandle _returnvalue))

(event Custom|SetHealthbyPlaylist
  ;[1 1 1] Health Scaling Logic
  ;;[1 1 1] Getting the team size from the playlist to set the team size variable
  (bind (_gamestate _playlist) (Class|FortAsyncActionCurrentPlaylistReady|PlaylistDataReadyAsync)
    (:OnReady
      (Variables|Default|SetPlaylistTeamSize (Math|Integer|Clamp(Integer) (Variables|/Script/FortniteGame.FortPlaylist|GetMaxTeamSize _playlist) 1 4))
      (CallFunction|ScaleHealthbyTeamSize))))

(event Custom|ScaleHealthByTeamSize
  ;[1 1 1] Health Scaling Logic
  ;;[1 1 1] Setting Health Values based on the TeamSize
  (bind _returnvalue (GameplayEffects|ApplyGameplayEffectToSelf :self (Ability|GetAbilitySystemComponent self) :GameplayEffectClass (Variables|Default|GetGE_ScaleHealthByTeamSize) :Level (Math|Conversions|ToFloat(Integer) (- (Variables|Default|GetPlaylistTeamSize) 1))))
  (Variables|Default|SetScaleHealthGEHandle _returnvalue))

(event Custom|HideHealthBar
  (HealthBarComponent|SetHealthBarVisibility (Variables|Default|GetFortHealthBar) false))

(event EventOnDamagePlayEffects
  (params (Damage :type Float%20%28single-precision%29) (DamageTags :type Gameplay%20Tag%20Container%20Structure%20%28by%20ref%29) (Momentum :type Vector :default "0, 0, 0") (HitInfo :type Hit%20Result%20Structure%20%28by%20ref%29) (InstigatedBy :type Fort%20Pawn%20Object%20Reference) (DamageCauser :type Actor%20Object%20Reference) (EffectContext :type Gameplay%20Effect%20Context%20Handle%20Structure))
  (|CallParentFunction Damage DamageTags Momentum HitInfo InstigatedBy DamageCauser EffectContext)
  (Default|PlayHitmarkerAudio InstigatedBy HitInfo))

(event Custom|StartStormChecking
  ;[1 1 1] Handling Storm Damage
  (Utilities|Time|SetTimerbyEvent (delegate Custom|CheckInsideStorm) (ScalableFloat|GetValueatLevel (Variables|Default|GetRow_StormCheckTimerTick)) true))

(event Custom|CheckInsideStorm
  (bind _self self)
  (bind _returnvalue (Ability|GetAbilitySystemComponent _self))
  ;[1 1 1] Handling Storm Damage
  (bind _returnvalue_2 (SafeZone|IsLocationinSafeZone (Transformation|GetActorLocation _self)))
  (if _returnvalue_2
    (GameplayEffects|RemoveActiveGameplayEffect _returnvalue (Variables|Default|GetGEHandle_StormDamage) -1)
    (else
      (bind _returnvalue_3 (GameplayEffects|ApplyGameplayEffectToSelf :self _returnvalue :GameplayEffectClass (Variables|Default|GetGE_StormDamage) :Level 0.0))
      (bind _output_get (Variables|Default|SetGEHandle_StormDamage _returnvalue_3))
      (GameplayEffects|UpdateActiveGameplayEffectSetbyCallerMagnitude (Ability|GetAbilitySystemComponent 0) _output_get "(TagName=\"SetByCaller.StormCampingDamage\")" 1.0))))

(fn PlayHitmarkerAudio
  (params (Instigator :type Fort%20Pawn%20Object%20Reference) (HitInfo :type Hit%20Result%20Structure))
  (bind (_bblockinghit _binitialoverlap _time _distance _location _impactpoint) (Collision|BreakHitResult HitInfo))
  (if (Pawn|IsLocallyViewed Instigator)
    ;[0.6 0.4 0.8] Instigator Hitmarker Audio
    (Audio|PlaySoundatLocation (select (FortAudioOptions|GetNewHitmarkersEnabled (BlueprintContext|GetFortSettingsContext)) "/HotAirBalloon/Assets/Audio/MSS_HotAirBalloon_Hitmarker_Instigator.MSS_HotAirBalloon_Hitmarker_Instigator" "/Game/Athena/Sounds/HitNotifications/HitMarker_Vehicle_Default_Cue.HitMarker_Vehicle_Default_Cue") _impactpoint)))

(fn ConstructionScript)
