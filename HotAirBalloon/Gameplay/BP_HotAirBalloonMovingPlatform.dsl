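; BP_HotAirBalloonMovingPlatform is a moving-platform component for the hot air balloon that supplies a toss velocity relative to its owner's travel direction.
; GetTossVelocity builds a vector from three scalable-float X, Y and Z values.
; It rotates that vector by a yaw taken from the owner's velocity angle, offset by 180 degrees.
; BeginPlay and Tick are empty, and what consumes the velocity is outside this graph.
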(event EventBeginPlay)

(event EventTick
  (params (DeltaSeconds :type Float%20%28single-precision%29)))

(fn GetTossVelocity
  (return (Math|Vector|RotateVector (Math|Vector|MakeVector (ScalableFloat|GetValueatLevel (Variables|Default|GetTossXVelocity)) (ScalableFloat|GetValueatLevel (Variables|Default|GetTossYVelocity)) (ScalableFloat|GetValueatLevel (Variables|Default|GetTossZVelocity))) (Math|Rotator|MakeRotator 0.0 0.0 (- (Fortnite|VectortoNormalizedAngleinDegrees (Transformation|GetVelocity (Components|GetOwner))) 1.8e+02)))))
