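; Game state component that randomizes the travel direction of hot air balloons when play begins.
; On the authority only, and only if a scalable-float toggle is enabled, it rolls and stores a single random boolean as the balloon direction.
; It then finds every CH7_HotAirBalloon actor and sets that shared direction through each one's train manager, as the balloons use the train rail system.
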(event EventBeginPlay
  (Networking|SwitchHasAuthority
    (:Authority
      (if (ScalableFloat|GetValueasBool (Variables|Default|GetShouldRandomizeBalloonDirection))
        (Variables|Default|SetBalloonDirection (Math|Random|RandomBool))
        (bind _outactors (Actor|GetAllActorsOfClass "/HotAirBalloon/Gameplay/CH7_HotAirBalloon.CH7_HotAirBalloon_C"))
        (for _array_element _outactors
          (Class|FortTrainManager|SetRailDirection (Class|TrainCar|GetTrainManager _array_element) (Variables|Default|GetBalloonDirection)))))))
