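; Loot container Blueprint for the hot air balloon that removes itself from the world when looted.
; After parent BeginPlay, it waits for playlist data to be ready, then destroys the inherited ItemMesh component, which a note, referring to a train, describes as a performance saving.
; Death effects call the parent and fire a local burst GameplayCue from the destroyed cue tag.
; Tick, overlap and server death events are empty.
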(event Collision|EventActorBeginOverlap
  (params (OtherActor :type Actor%20Object%20Reference)))

(event EventTick
  (params (DeltaSeconds :type Float%20%28single-precision%29)))

(event EventBeginPlay
  (|CallParentFunction)
  (Class|FortAsyncActionCurrentPlaylistReady|PlaylistDataReadyAsync
    (:OnReady
      ;[1 0.1135 0] Destroy components from Parent asset to save on perf for the train.
      ; Note: We will try to move the sound to the train car and coordinate destroying it when the loot container dies.
      (Components|DestroyComponent (Variables|Default|GetItemMesh)))))

(event EventOnLoot
  (Actor|DestroyActor))

(event EventOnDeathPlayEffects
  (params (Damage :type Float%20%28single-precision%29) (DamageTags :type Gameplay%20Tag%20Container%20Structure%20%28by%20ref%29) (Momentum :type Vector :default "0, 0, 0") (HitInfo :type Hit%20Result%20Structure%20%28by%20ref%29) (InstigatedBy :type Fort%20Pawn%20Object%20Reference) (DamageCauser :type Actor%20Object%20Reference) (EffectContext :type Gameplay%20Effect%20Context%20Handle%20Structure))
  (|CallParentFunction Damage DamageTags Momentum HitInfo InstigatedBy DamageCauser EffectContext)
  (GameplayCue|ExecuteGameplayCue%28Burst%29Local self (Variables|Default|GetDestroyed%20Gameplay%20Cue%20Tag) (Ability|GameplayCue|MakeGameplayCueParameters 0.0)))

(event EventOnDeathServer
  (params (Damage :type Float%20%28single-precision%29) (DamageTags :type Gameplay%20Tag%20Container%20Structure%20%28by%20ref%29) (Momentum :type Vector :default "0, 0, 0") (HitInfo :type Hit%20Result%20Structure%20%28by%20ref%29) (InstigatedBy :type Controller%20Object%20Reference) (DamageCauser :type Actor%20Object%20Reference) (EffectContext :type Gameplay%20Effect%20Context%20Handle%20Structure)))

(fn ConstructionScript
  (|CallParentFunction))
