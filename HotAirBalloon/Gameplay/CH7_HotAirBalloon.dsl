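; Hot air balloon that travels between stations through the Fort train system, carrying players as riders.
; Station events drive arrival and departure cues, audio and healing of an attached, server-spawned health actor whose death destroys the balloon.
; A server rider map and timer start skydiving after fall delay or height, alongside debase and destruction triggers.
; It also launches pawns, impulses vehicles, snaps cosmetic yaw to 90 degrees and detaches inessential components at low fidelity.
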(event EventOnTrainManagerReady
  (params (TrainManager :type Fort%20Train%20Manager%20Object%20Reference))
  ;[1 1 1] Handling Balloon arriving at a station
  (EventDispatchers|BindEventtoOnTrainStationArrivalIntervalReachedEvent (Variables|/Script/TrainCoreRuntime.FortTrainManager|GetRail TrainManager) (delegate Custom|TrainArrived))
  ;[1 1 1] Handling Balloon leaving the station
  (EventDispatchers|BindEventtoOnTrainStationDepartureEvent (Variables|/Script/TrainCoreRuntime.FortTrainManager|GetRail TrainManager) (delegate Custom|LeavingStation)))

(event Custom|TrainArrived
  (params (Train :type Fort%20Train%20Manager%20Object%20Reference) (StationIdentifier :type Gameplay%20Tag%20Structure) (TimeUntilArrival :type Float%20%28single-precision%29) (Interval :type Float%20%28single-precision%29))
  (bind _returnvalue (CallFunction|GetTrainManager))
  ;[1 1 1] Handling Balloon arriving at a station
  (if (and (== Train _returnvalue) (== TimeUntilArrival 0))
    (Variables|Default|SetIsAtStation true)
    (CallFunction|ApplyBalloonHealing)
    (GameplayCue|ExecuteGameplayCue self (Variables|Default|GetGCT_BalloonArrived) (Class|AthenaFunctionLibrary|MakeSimpleFortEffectContext self))
    (elif (and (== Train _returnvalue) (== TimeUntilArrival 15.0))
      (CallFunction|PlayBalloonStationAudio true (Variables|Default|GetGCT_BalloonArriving)))))

(event Custom|LeavingStation
  (params (Train :type Fort%20Train%20Manager%20Object%20Reference) (StationIdentifier :type Gameplay%20Tag%20Structure) (TimeUntilNextArrival :type Float%20%28single-precision%29))
  ;[1 1 1] Handling Balloon leaving the station
  (if (== Train (CallFunction|GetTrainManager))
    (Variables|Default|SetIsAtStation false)
    (CallFunction|PlayBalloonStationAudio false (Variables|Default|GetGCT_BalloonDeparting))))

(event OnPawnBasedonMovingPlatform(BP_HotAirBalloonMovingPlatform)
  (params (MovingPlatform :type Fort%20Base%20Moving%20Platform%20Component%20Object%20Reference) (Pawn :type Fort%20Pawn%20Object%20Reference))
  (Networking|SwitchHasAuthority
    (:Authority
      (bind _asfort_player_pawn_athena (Utilities|Casting|CastToFortPlayerPawnAthena Pawn))
      (CallFunction|StartTrackingRider _asfort_player_pawn_athena))))

(event OnPawnDebasedMovingPlatform(BP_HotAirBalloonMovingPlatform)
  (params (MovingPlatform :type Fort%20Base%20Moving%20Platform%20Component%20Object%20Reference) (Pawn :type Fort%20Pawn%20Object%20Reference) (NewBase :type Actor%20Object%20Reference))
  (Networking|SwitchHasAuthority
    (:Authority
      (bind _asfort_player_pawn_athena (Utilities|Casting|CastToFortPlayerPawnAthena Pawn))
      (CallFunction|StopTrackingRider _asfort_player_pawn_athena)
      (if (Math|Boolean|ANDBoolean (ScalableFloat|GetValueasBool (Variables|Default|GetRow_ShouldApplySkydivingOnDebase)) (GameplayTags|HasMatchingGameplayTag _asfort_player_pawn_athena "(TagName=\"MovementMode.Falling\")") (CallFunction|ShouldAllowSkydiving _asfort_player_pawn_athena))
        (Parachute|BeginSkydiving _asfort_player_pawn_athena false)))))

(event Custom|OnRiderMovementModeChanged
  (params (Character :type Character%20Object%20Reference) (PrevMovementMode :type EMovementMode%20Enum :default "MOVE_None") (PreviousCustomMode :type Byte))
  (bind _asfort_player_pawn_athena (Utilities|Casting|CastToFortPlayerPawnAthena Character))
  (if (Utilities|Enum|Equal(Enum) (Variables|/Script/Engine.CharacterMovementComponent|GetMovementMode (Variables|/Script/Engine.Character|GetCharacterMovement _asfort_player_pawn_athena)) "MOVE_Falling")
    (if (not (Utilities|Enum|Equal(Enum) PrevMovementMode "MOVE_Falling"))
      ;[1 1 1] If we're falling now and we weren't before, track the time we began falling so we know when to enter skydiving state
      (Utilities|Map|Add (Variables|ActiveRiders|GetActiveRiders) _asfort_player_pawn_athena (Utilities|Struct|MakeHotAirBalloonRiderState true (Fortnite|Sync|Time|GetServerWorldTimeSeconds) (Transformation|GetActorLocation _asfort_player_pawn_athena))))
    (else
      (Utilities|Map|Add (Variables|ActiveRiders|GetActiveRiders) _asfort_player_pawn_athena))))

(event EventBeginPlay
  ;[1 1 1] Assign a random angle to the balloon, but ensure it always remains 90-degree locked so it lines up correctly at stations
  (Networking|SwitchHasAuthority
    (:Authority
      (Variables|Default|SetCosmeticAngle (Math|Conversions|ToFloat(Integer) (* (Math|Random|RandomIntegerinRange 0 3) 90)))
      (bind _returnvalue (Game|SpawnActorfromClass "/HotAirBalloon/Gameplay/BGA_BalloonHealth.BGA_BalloonHealth_C" (Transformation|GetActorTransform self) "Undefined" "MultiplyWithRoot" 0 (Variables|Default|GetBurstColor) 0 0))
      (bind _output_get (Variables|Default|SetBalloonHealth _returnvalue))
      (Transformation|AttachActorToActor _output_get self "None" "SnapToTarget" "SnapToTarget" "SnapToTarget" true)
      (EventDispatchers|BindEventtoOnDied (Variables|Default|GetBalloonHealth) (delegate Custom|DestroyBalloon)))))

(event Custom|UpdateActiveRiders
  (bind (_keypin _valuepin) (Utilities|Map|ForEachLoop(Map) (Variables|ActiveRiders|GetActiveRiders)))
  (if (and (Utilities|Struct|BreakHotAirBalloonRiderState _valuepin) (CallFunction|ShouldAllowSkydiving _keypin))
    (if (or (>= (- (Fortnite|Sync|Time|GetServerWorldTimeSeconds) (.FallingStartTimestamp_2_DD4034DD40C28C6661052EBD2671D2D6 (Utilities|Struct|BreakHotAirBalloonRiderState _valuepin))) (ScalableFloat|GetValueatLevel (Variables|ActiveRiders|GetSkydiveFallDelay))) (>= (- (.z (.FallingStartPosition_5_FF8FF31B4E158839E0E0CF8E223E2454 (Utilities|Struct|BreakHotAirBalloonRiderState _valuepin))) (.z (Transformation|GetActorLocation _keypin))) (ScalableFloat|GetValueatLevel (Variables|ActiveRiders|GetSkydiveFallHeight))))
      (Parachute|BeginSkydiving _keypin false)
      (CallFunction|StopTrackingRider _keypin))))

(event EventOnActorTargetedForImpulse
  (params (Actor :type Actor%20Object%20Reference))
  ; OnActorTargetedforImpulse - If NOT specter (ghost), Apply Front Impulse
  (bind _asfort_player_pawn_athena (Utilities|Casting|CastToFortPlayerPawnAthena Actor)
    (:then
      (Utilities|IsValid (Ability|GetAbilitySystemComponent _asfort_player_pawn_athena)
        (:Is%20Valid
          (if (not (GameplayTags|HasMatchingGameplayTag (Ability|GetAbilitySystemComponent _asfort_player_pawn_athena) (Variables|Default|GetCheckHasSpecterTag)))
            _))
        (:Is%20Not%20Valid
          _)) ; If players are not currently possessed then ASC will return nullptr.
      (Default|ApplyFrontImpulse Actor) ; Send impulse to non-ghost actors
    )
    (:CastFailed
      _)))

(event Custom|DestroyBalloon
  (params (DamagedActor :type Actor%20Object%20Reference) (Damage :type Float%20%28single-precision%29) (InstigatedBy :type Controller%20Object%20Reference) (DamageCauser :type Actor%20Object%20Reference) (HitLocation :type Vector) (FHitComponent :type Primitive%20Component%20Object%20Reference) (BoneName :type Name) (Momentum :type Vector))
  (bind _keys (Utilities|Map|Keys (Variables|ActiveRiders|GetActiveRiders)))
  (for _array_element _keys
    (if (or (not (ScalableFloat|GetValueasBool (Variables|Default|GetCheckAllowSkydiveOnDestroy))) (CallFunction|ShouldAllowSkydiving _array_element))
      (Parachute|BeginSkydiving _array_element false)
      (elif (ScalableFloat|GetValueasBool (Variables|Default|GetAlwaysUntrackRiderOnDestroy))
        _))
    (CallFunction|StopTrackingRider _array_element))
  (Development|PrintString "DestroyBalloon")
  (Actor|DestroyActor self))

(event EventOnActorTargetedForDestruction
  (params (Actor :type Actor%20Object%20Reference))
  ; If we have an actor targeted for destruction that wasn't destroyed, impulse it instead
  (Default|ApplySurroundingImpulse Actor))

(event Custom|ApplyBalloonHealing
  (bind _balloonhealth (Variables|Default|GetBalloonHealth))
  (bind _returnvalue (Ability|GetAbilitySystemComponent _balloonhealth))
  ;[1 1 1] Balloon self healing when arriving at a station
  (bind _returnvalue_2 (Ability|GameplayEffect|AssignTagSetbyCallerMagnitude (GameplayEffects|MakeOutgoingSpec :self _returnvalue :GameplayEffectClass (Variables|Default|GetGE_BalloonHealing) :Level 0.0) "(TagName=\"SetByCaller.Healing\")" (* (Health|GetMaxHealth _balloonhealth :!class /Script/FortniteGame.BuildingActor) (ScalableFloat|GetValueatLevel (Variables|Default|GetRow_BalloonHealPercentPerSecond)))))
  (bind _returnvalue_3 (GameplayEffects|ApplyGameplayEffectSpecToSelf _returnvalue _returnvalue_2))
  (Variables|Default|SetGEHandle_BalloonHealing _returnvalue_3))

(event Custom|RemoveBalloonHealing
  ;[1 1 1] Balloon self healing when arriving at a station
  (GameplayEffects|RemoveActiveGameplayEffect (Ability|GetAbilitySystemComponent (Variables|Default|GetBalloonHealth)) (Variables|Default|GetGEHandle_BalloonHealing) -1))

(event EventOnAVFidelityChanged
  (params (bHighFidelity :type Boolean))
  (bind _audiomovement (Variables|Default|GetAudioMovement))
  (bind _ph_lootcontainer (Variables|Default|GetPH_LootContainer))
  (bind _ph_lootcontainer2 (Variables|Default|GetPH_LootContainer2))
  (bind _wind_fx (Variables|Default|GetWind_FX))
  (bind _audiobells (Variables|Default|GetAudioBells))
  (bind _staticmesh (Variables|Default|GetStaticMesh))
  ; When on Low AV fidelity, detach inessential audio and visual components so that they aren't updated with balloon movement
  (Components|Activation|SetActive (Variables|Default|GetAudioMovement) bHighFidelity false)
  (Components|Activation|SetActive (Variables|Default|GetPH_LootContainer) bHighFidelity false)
  (Components|Activation|SetActive (Variables|Default|GetPH_LootContainer2) bHighFidelity false)
  (Components|Activation|SetActive (Variables|Default|GetWind_FX) bHighFidelity false)
  (Components|Activation|SetActive (Variables|Default|GetAudioBells) bHighFidelity false)
  (if bHighFidelity
    (Transformation|AttachComponentToComponent _audiomovement _staticmesh "None" "KeepRelative" "KeepRelative" "KeepRelative" false)
    (Transformation|AttachComponentToComponent _ph_lootcontainer _staticmesh "None" "KeepRelative" "KeepRelative" "KeepRelative" false)
    (Transformation|AttachComponentToComponent _ph_lootcontainer2 _staticmesh "None" "KeepRelative" "KeepRelative" "KeepRelative" false)
    (Transformation|AttachComponentToComponent _wind_fx _staticmesh "None" "KeepRelative" "KeepRelative" "KeepRelative" false)
    (Transformation|AttachComponentToComponent _audiobells _staticmesh "None" "KeepRelative" "KeepRelative" "KeepRelative" false)
    (else
      (Transformation|DetachFromComponent _audiomovement "KeepRelative" "KeepRelative" "KeepRelative" true)
      (Transformation|DetachFromComponent _ph_lootcontainer "KeepRelative" "KeepRelative" "KeepRelative" true)
      (Transformation|DetachFromComponent _ph_lootcontainer2 "KeepRelative" "KeepRelative" "KeepRelative" true)
      (Transformation|DetachFromComponent _wind_fx "KeepRelative" "KeepRelative" "KeepRelative" true)
      (Transformation|DetachFromComponent _audiobells "KeepRelative" "KeepRelative" "KeepRelative" true))))

(fn ApplyFrontImpulse
  (params (Actor :type Actor%20Object%20Reference))
  (bind _asfort_pawn (Utilities|Casting|CastToFortPawn Actor))
  (Character|LaunchCharacter _asfort_pawn (Math|Vector|MakeVector 0.0 0.0 (ScalableFloat|GetValueatLevel (Variables|Default|GetLaunchSpeed))) true true) ; We need this initial launch since the knockback GE/GA takes too long to activate, and our character could already be on the train. GE/GA still handles knockback status so it handles cancelling other abilities
  (Fortnite|ApplyGameplayEffectsToActor _asfort_pawn (Variables|Default|GetEffects%20to%20Apply) (Transformation|GetActorLocation _asfort_pawn) false 0 (Ability|GetFortAbilitySystemComponent self))
  (GameplayCue|ExecuteGameplayCue _asfort_pawn (Variables|Default|GetPlayerLaunchGameplayCueTag))
  (bind _asfort_player_pawn_athena (Utilities|Casting|CastToFortPlayerPawnAthena Actor))
  (if (and (ScalableFloat|GetValueasBool (Variables|Default|GetRefreshSkydiveOnImpulse)) (.ReturnValue (Utilities|Map|Find (Variables|ActiveRiders|GetActiveRiders) _asfort_player_pawn_athena)))
    ; Refresh the skydive state if we've been impulsed by the balloon
    (Utilities|Map|Add (Variables|ActiveRiders|GetActiveRiders) _asfort_player_pawn_athena (Utilities|Struct|MakeHotAirBalloonRiderState true (Fortnite|Sync|Time|GetServerWorldTimeSeconds) (Transformation|GetActorLocation)))))

(fn ApplySurroundingImpulse
  (params (Actor :type Actor%20Object%20Reference))
  ; Hot Air Balloons only care about impulsing vehicles via general destruction targeting
  (bind _asfort_vehicle_interface (Utilities|Casting|CastToFortVehicleInterface Actor))
  (Vehicle|WakeUp _asfort_vehicle_interface)
  (Ability|Advanced|GrantandActivateAbilityonSelf (Ability|GetAbilitySystemComponent Actor) (Variables|Default|GetVehicle%20Impulse%20Ability) self))

(fn OnRep_CosmeticAngle
  (Transformation|SetWorldRotation (Variables|Default|GetStaticMesh) (Math|Rotator|MakeRotator 0.0 0.0 (Variables|Default|GetCosmeticAngle)) false false))

(fn PlayBalloonStationAudio
  (params (bIsArriving :type Boolean) (GameplayCueTag :type Gameplay%20Tag%20Structure))
  (GameplayCue|ExecuteGameplayCue self GameplayCueTag (Class|AthenaFunctionLibrary|MakeSimpleFortEffectContext self))
  (if bIsArriving
    (Utilities|FlowControl|Sequence
      (:then_0
        (bind _audiomovement (Variables|Default|GetAudioMovement)
          (:then
            (Audio|Components|Audio|AdjustVolume _audiomovement 2.0 0.25 "Linear" :!class /Script/Engine.AudioComponent))
          (:else)))
      (:then_1
        (bind _audiobells (Variables|Default|GetAudioBells)
          (:then
            (Audio|Components|Audio|AdjustVolume _audiobells 1.0 1.0 "Linear" :!class /Script/Engine.AudioComponent))
          (:else))))
    (else
      (Utilities|FlowControl|Sequence
        (:then_0
          (bind _audiomovement_2 (Variables|Default|GetAudioMovement)
            (:then
              (Audio|Components|Audio|AdjustVolume _audiomovement_2 1.0 1.0 "Linear" :!class /Script/Engine.AudioComponent))
            (:else)))
        (:then_1
          (bind _audiobells_2 (Variables|Default|GetAudioBells)
            (:then
              (Audio|Components|Audio|AdjustVolume _audiobells_2 2.0 1.0 "Linear" :!class /Script/Engine.AudioComponent))
            (:else)))))))

(fn ShouldAllowSkydiving
  (params (Player :type Fort%20Player%20Pawn%20Athena%20Object%20Reference))
  (Utilities|IsValid Player
    (:Is%20Valid
      (return (and (Math|Boolean|ANDBoolean (and (ScalableFloat|GetValueasBool (Variables|Default|GetCanEverSkydive)) (not (Variables|Default|GetIsAtStation))) (or (not (ScalableFloat|GetValueasBool (Variables|Default|GetPreventSkydiveWithKnockback))) (not (GameplayTags|HasAnyMatchingGameplayTags Player (Variables|Default|GetKnockbackTagContainer)))) (not (Health|IsDBNO Player))) (or (not (ScalableFloat|GetValueasBool (Variables|Default|GetCheckForSkydiveBlockedTags))) (not (GameplayTags|HasAnyMatchingGameplayTags Player (Variables|Default|GetBlockSkydiveTagContainer)))))))) ; Returning \n"Attempted to access PlayerPawn_Athena_C_0 via property Player".\nIn the case where SpecterPawn enters balloon
)

(fn ShouldShowSoundIndicator
  (params (PlayerController :type Fort%20Player%20Controller%20Object%20Reference))
  (bind _returnvalue (Game|Player|GetPlayerPawn PlayerController))
  (Utilities|IsValid _returnvalue
    (:Is%20Valid
      (return (not (== (Pawn|Components|CharacterMovement|GetMovementBaseObject (Variables|/Script/Engine.Character|GetCharacterMovement _returnvalue)) (Variables|Default|GetStaticMesh)))))
    (:Is%20Not%20Valid
      (return false))))

(fn StartTrackingRider
  (params (Player :type Fort%20Player%20Pawn%20Athena%20Object%20Reference))
  (bind _returnvalue (Ability|GetAbilitySystemComponent Player))
  (if (Utilities|Map|IsEmpty (Variables|ActiveRiders|GetActiveRiders))
    (bind _returnvalue_2 (Utilities|Time|SetTimerbyEvent (EventDispatchers|CreateEvent :!event UpdateActiveRiders) (ScalableFloat|GetValueatLevel (Variables|ActiveRiders|GetActiveRidersUpdateFrequency)) true))
    (Variables|ActiveRiders|SetUpdateActiveRidersHandle _returnvalue_2)
    (else
      _))
  (Utilities|Map|Add (Variables|ActiveRiders|GetActiveRiders) Player)
  (Character|BindEventtoMovementModeChangedDelegate Player (EventDispatchers|CreateEvent :!event OnRiderMovementModeChanged))
  (Class|FortTrainManager|BeginMarkingNextStation (CallFunction|GetTrainManager) (Fortnite|GetFortPlayerControllerfromActor Player))
  (if (ScalableFloat|GetValueasBool (Variables|Default|GetRow_ShouldApplyFallDamageImmunity))
    (GameplayEffects|ApplyGameplayEffectToSelf :self _returnvalue :GameplayEffectClass (Variables|Default|GetGE_FallDamageImmunity) :Level 0.0)
    (else
      _))
  (if (ScalableFloat|GetValueasBool (Variables|Default|GetRow_ShouldBlockJamTracks))
    (GameplayEffects|ApplyGameplayEffectToSelf :self _returnvalue :GameplayEffectClass (Variables|Default|GetGE_BlockJamTracks) :Level 0.0)))

(fn StopTrackingRider
  (params (Player :type Fort%20Player%20Pawn%20Athena%20Object%20Reference))
  (bind _returnvalue (Ability|GetAbilitySystemComponent Player))
  (bind _activeriders (Variables|ActiveRiders|GetActiveRiders))
  (Utilities|Map|Remove _activeriders Player)
  (if (Utilities|Map|IsEmpty _activeriders)
    (Utilities|Time|ClearandInvalidateTimerbyHandle (Variables|ActiveRiders|GetUpdateActiveRidersHandle)))
  (Character|UnbindEventfromMovementModeChangedDelegate Player (EventDispatchers|CreateEvent :!event OnRiderMovementModeChanged))
  (Class|FortTrainManager|EndMarkingNextStation (CallFunction|GetTrainManager) (Fortnite|GetFortPlayerControllerfromActor Player))
  (GameplayEffects|RemoveActiveGameplayEffectbySourceEffect _returnvalue (Variables|Default|GetGE_FallDamageImmunity) _returnvalue -1)
  (GameplayEffects|RemoveActiveGameplayEffectbySourceEffect _returnvalue (Variables|Default|GetGE_BlockJamTracks) _returnvalue -1))

(fn ConstructionScript
  (|CallParentFunction))
